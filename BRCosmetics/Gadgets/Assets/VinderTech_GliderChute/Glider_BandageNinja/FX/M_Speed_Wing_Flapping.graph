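# Unlit, two-sided translucent Niagara material for wing-flap speed streaks on the BandageNinja glider.
# Two counter-panning smoke textures are screen-blended, then tinted by particle color and time-of-day colors for emissive.
# Opacity combines that noise and particle alpha with UV edge falloff, view-angle fade and depth fade, which is bypassed on lower quality and forward/mobile paths.
# WingTipValue and InterpolatedValue dynamic parameters bound a reflected V-axis gradient that also shapes opacity.
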Material /BRCosmetics/Gadgets/Assets/VinderTech_GliderChute/Glider_BandageNinja/FX/M_Speed_Wing_Flapping
graph {
 properties {
  bUsedWithNiagaraMeshParticles=true
  bUsedWithNiagaraSprites=true
  blendMode=BLEND_Translucent
  pixelDepthOffsetMode=PDOM_Legacy
  shadingModel=MSM_Unlit
  twoSided=true
  usedShadingModels=Unlit
 }
 nodes [
  n1 = TextureCoordinate() {uTiling=8, vTiling=0.35}
  n2 = Constant2Vector() {g=1.5, r=0.25}
  n3 = TextureSample(UVs=Panner(Coordinate=Multiply(A=@n1, B=@n2)) {speedX=0.25, speedY=0.7}) {texture=&T_Smooth_tile}
  n4 = TextureCoordinate() {uTiling=6, vTiling=0.2}
  n5 = TextureSample(UVs=Panner(Coordinate=Multiply(A=@n4, B=@n2)) {speedX=-0.25, speedY=-1}) {texture=&T_Inky_Smoke_Tile}
  LinearInterpolate(A=@n3.R, B=@n5.R)
  n6 = LinearInterpolate(B=MaterialFunctionCall(
    Color_Morning=VectorParameter() {parameterName=MorningTint, defaultValue=<rgba 1.15 0.8884 0.805 0>}.RGB
    Color_Day=VectorParameter() {parameterName=DayTint, defaultValue=<rgba 1.0812 1.2 1.1958 0>}.RGB
    Color_Evening=VectorParameter() {parameterName=EveningTint, defaultValue=<rgba 1.1 0.7845 0.77 1>}.RGB
    Color_Night=VectorParameter() {parameterName=NightTint, defaultValue=<rgba 0.2656 0.85 0.85 0>}.RGB
   ) {materialFunction=&Color_Time_of_Day}.Result) {constA=1}
  n7 = ParticleColor()
  n8 = MaterialFunctionCall!(Blend=@n3.R, Base=@n5.R) {materialFunction=/Engine/Functions/Engine_MaterialFunctions03/Blends/Blend_Screen}
  MaterialFunctionCall(Emissive=Multiply(A=Multiply(A=@n6, B=@n7.RGB), B=@n8.Result)) {materialFunction=&MF_FX_EmissiveAdjustments}
  n9 = Multiply(A=@n7.A, B=@n8.Result)
  n10 = TextureCoordinate()
  n11 = Multiply(A=OneMinus(None=ComponentMask(None=@n10) {g=true}), B=ComponentMask(None=@n10) {g=true})
  n12 = Multiply(A=OneMinus(None=ComponentMask(None=@n10) {r=true}), B=ComponentMask(None=@n10) {r=true})
  n13 = Power(Base=Multiply(A=Multiply(A=@n11, B=@n12)) {constB=10})
  n14 = Constant() {r=2}
  n15 = DepthFade(FadeDistance=ScalarParameter() {defaultValue=33, parameterName=Depth Fade Distance})
  n16 = Constant() {r=1}
  n17 = ShadingPathSwitch(Default=@n15, Deferred=@n15, Forward=@n16, Mobile=@n16)
  n18 = DynamicParameter() {paramNames [WingTipValue, InterpolatedValue, "", ""]}
  n19 = Saturate(None=Divide(A=Distance(A=@n18.WingTipValue, B=@n18.WingTipValue), B=ScalarParameter() {defaultValue=0.333, parameterName=Distance Softness}))
  n20 = ComponentMask(None=TextureCoordinate()) {g=true}
  n21 = ConstantBiasScale(None=@n18.WingTipValue)
  n22 = ConstantBiasScale(None=@n18.InterpolatedValue)
  n23 = Saturate(None=Divide(A=Subtract(A=@n20, B=@n21), B=Subtract(A=@n22, B=@n21)))
  n24 = MaterialFunctionCall!(Gradient=Frac(None=OneMinus(None=@n23)), Center Point=ScalarParameter() {defaultValue=0.8, parameterName=Gradient Center Point}) {materialFunction=/Game/Effects/Fort_Effects/MaterialFunctions/MF_GradientReflected}
  Multiply(
   A=Multiply(
    A=Multiply(A=@n9, B=@n13)
    B=MaterialFunctionCall("FadeSensitivity (optional)"=@n14) {materialFunction=&FadeBasedOnViewAngle}.Result
   )
   B=Multiply(A=QualitySwitch(Default=@n17, Low=@n16, High=@n17, Medium=@n16, Epic=@n17), B=Multiply(A=@n19, B=@n24.Result))
  )
 ]
}
