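# Niagara effect for the Bandage Ninja glider, combining tumbling feather meshes with ribbon trails.
# Feathers spawn in short staggered bursts, then drift under drag and vortex force while shrinking.
# Soft and regular anim-trail ribbons follow skeletal mesh bone pairs, faded by system opacity and, in gameplay, an integrator value.
# User.Variant picks colors and front-end modifiers retune motion; debug, speedline and second feather emitters are disabled.
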NiagaraSystem /BRCosmetics/Gadgets/Assets/VinderTech_GliderChute/Glider_BandageNinja/FX/NS_Glider_BandageNinja
dependencies {
 usedDataInterfaces [
  &NiagaraDataInterfaceVector2DCurve
  &NiagaraDataInterfaceCurve
  &NiagaraDataInterfaceColorCurve
  &NiagaraDataInterfaceArrayColor
  &NiagaraDataInterfaceVectorCurve
 ]
 usedDynamicInputs [
  /Niagara/DynamicInputs/Math/Min_Float
  &ClampFloat
  &Subtract_Float
  &Boolean_NotOperation
 ]
 usedModules [
  &SystemState
  &self:SystemSpawnScript.SystemScriptSource.SystemScriptGraph.NiagaraNodeAssignment_0.SetVariables_#1
  &self:SystemSpawnScript.SystemScriptSource.SystemScriptGraph.NiagaraNodeAssignment_6.SetVariables_#2
  &Timeline
  &self:SystemSpawnScript.SystemScriptSource.SystemScriptGraph.NiagaraNodeAssignment_1.SetVariables_#3
  &self:SystemSpawnScript.SystemScriptSource.SystemScriptGraph.NiagaraNodeAssignment_4.SetVariables_#4
  &self:SystemSpawnScript.SystemScriptSource.SystemScriptGraph.NiagaraNodeAssignment_2.SetVariables_#5
  &FN_Integrator
 ]
}
emitters {
 FN_Debug {
  bEnabled=false
  emitterUpdateScript [
   &EmitterState(Life Cycle Mode=::NewEnumerator1, Loop Behavior=::NewEnumerator1, Loop Duration=1, UseLoopDelay=false)
   &SpawnBurst_Instantaneous(Spawn Count=1, Use Spawn Probability=false)
  ]
  particleSpawnScript [/Niagara/Modules/Spawn/Initialization/InitializeParticle(
    Module.Write Color=true
    Module.Write Lifetime=true
    Module.Write Mass=true
    Module.Write Material Random=true
    Module.Write Position=false
    Module.Write Scale=false
    Module.Write SpriteRotation=false
    Module.Write SpriteSize=true
    Position=&SimulationPosition()
    Sprite Size=<100 100>
   )]
  particleUpdateScript [
   &ParticleState(Kill Particles When Lifetime Has Elapsed=false)
   &FN_Timer()
   /SetVariables(Particles.Debug:NiagaraFloat=$User.Variant) {isSetParameters=true}
   &DynamicMaterialParameters(
    Index 0 Param 1=$Particles.Debug
    Module.Param0WriteEnabled=true
    Module.Param1WriteEnabled=true
    Module.Param2WriteEnabled=true
    Module.Param3WriteEnabled=true
   )
   &CameraOffset(Camera Offset Amount=111)
  ]
  rendererClasses [&NiagaraSpriteRendererProperties]
  renderers [{rendererClass=&NiagaraSpriteRendererProperties}]
 }
 Feather {
  emitterUpdateScript [
   &EmitterState(Life Cycle Mode=::NewEnumerator1, Loop Behavior=::NewEnumerator1, Loop Duration=0.4, UseLoopDelay=false)
   &SpawnRate(SpawnRate=5, Use Spawn Probability=false)
   &SpawnBurst_Instantaneous(
    Spawn Count=&RandomRangeInt(Minimum=6, Maximum=9)
    Use Spawn Probability=false
    UseLoopCountLimit=false
   )
   &SpawnBurst_Instantaneous(
    Spawn Count=&RandomRangeInt(Minimum=1, Maximum=2)
    Spawn Time=0.15
    Use Spawn Probability=false
    UseLoopCountLimit=false
   ) {module=SpawnBurst_Instantaneous001}
   &SpawnBurst_Instantaneous(
    Spawn Count=&RandomRangeInt(Minimum=1, Maximum=2)
    Spawn Time=0.3
    Use Spawn Probability=false
    UseLoopCountLimit=false
   ) {module=SpawnBurst_Instantaneous002}
  ]
  particleSpawnScript [
   /Niagara/Modules/Spawn/Initialization/V2/InitializeParticle(
    Color=&LerpLinearColors(
     StartColor=&ColorFromCurve(
      ColorCurve=DataInterface() {propertyValues {
        AlphaCurve {keys [<at 0 1>]}
        BlueCurve {keys [<at 0 0.0581 RCIM_Cubic>,<at 0.5 0.0979>,<at 1 0.1 RCIM_Cubic>]}
        GreenCurve {keys [<at 0 0.0557 RCIM_Cubic>,<at 0.5 0.0937>,<at 1 0.0419 RCIM_Cubic>]}
        RedCurve {keys [<at 0 0.1 RCIM_Cubic>,<at 0.5 0.1>,<at 1 0.0358 RCIM_Cubic>]}
       }}
      CurveIndex=&RandomRangeFloat(Minimum=0, Maximum=1)
     )
     EndColor=&ColorFromCurve(ColorCurve=DataInterface() {propertyValues {
        AlphaCurve {keys [<at 0 1 RCIM_Cubic>]}
        BlueCurve {keys [<at 0 0.6308 RCIM_Cubic>,<at 1 0.1845 RCIM_Cubic>]}
        GreenCurve {keys [<at 0 0.624 RCIM_Cubic>,<at 1 0.1529 RCIM_Cubic>]}
        RedCurve {keys [<at 0 0.6584 RCIM_Cubic>,<at 1 0.1714 RCIM_Cubic>]}
       }})
     LerpFactor=$User.Variant
    )
    Lifetime=&RandomRangeFloat(Minimum=1.3, Maximum=2.2)
    Mesh Scale=&Multiply_Vector(
     A=$Engine.Owner.Scale
     B=/Niagara/DynamicInputs/CascadeConversion/SplitVector(
      Double Component Vector Channels:ECascadeNiagaraTwoVectorChannels=::NewEnumerator0
      Double Component:NiagaraFloat=&RandomRangeFloat(Minimum=0.3, Maximum=0.5)
      Single Component:NiagaraFloat=&RandomRangeFloat(Minimum=0, Maximum=1)
     )
    )
    Mesh Scale Mode=::NewEnumerator4
    UsePositionOffset=false
    Write Material Random=true
   )
   &ShapeLocation(
    MaskSpawnGroup=false
    Offset=&RandomRangeVector(Minimum=<-170 -70 -30>, Maximum=<170 -30 -20>)
    Offset Coordinate Space=::Local
    Offset Mode=::NewEnumerator0
    Override Seed=false
    Sphere Radius=5
    Sphere Surface Distribution=1
    UseOwnerScale=true
   )
   &AddVelocity(
    Rotation Coordinate Space=::Simulation
    Velocity=&FN_FrontEndVector3Modifier(
     Gameplay=&RandomRangeVector(Minimum=<-150 -150 -50>, Maximum=<150 150 -800>)
     Front End=&RandomRangeVector(Minimum=<-150 -150 -222>, Maximum=<150 150 222>)
    )
   )
   &AddVelocity(
    Rotation Coordinate Space=::Simulation
    Velocity=&Multiply_VectorByFloat(Vector=$Particles.ShapeLocation.ShapeNormal, Float=50)
   ) {module=AddVelocity001}
   &DynamicMaterialParameters(
    Index 0 Param 1=&RandomRangeFloat(Minimum=0, Maximum=0.9)
    Index 0 Param 2=9
    Index 0 Param 4=&RandomRangeFloat(Minimum=0, Maximum=1)
    Module.Param0WriteEnabled=true
    Module.Param1WriteEnabled=true
    Module.Param2WriteEnabled=true
    Module.Param3WriteEnabled=true
   )
   &InitialMeshOrientation(Use Rotation Vector=false)
   /SetVariables(Particles.SpawnRotationVector=&RandomRangeVector(Minimum=<0 0 0>, Maximum=<1 1 1>)) {isSetParameters=true, module=SetVariables_#6}
  ]
  particleUpdateScript [
   &ParticleState()
   &UpdateMeshOrientation(
    Rotation Rate=&FloatFromCurve(FloatCurve=DataInterface() {propertyValues {Curve {keys [<at 0 1>,<at 1 0>]}}}, Scale Curve=2)
    Rotation Vector=$Particles.SpawnRotationVector
   )
   &ScaleMeshSize(Scale Factor=&VectorFromFloat(Value=&FloatFromCurve(FloatCurve=DataInterface() {propertyValues {Curve {keys [<at 0.5 1>,<at 1 0>]}}})))
   &Drag(Drag=2, Use Linear Drag=true, Use Rotational Drag=false)
   &VortexForce(
    UseOriginOffset=true
    Vortex Axis=<1 1 1>
    Vortex Axis Coordinate Space:ENiagaraCoordinateSpace=::Local
    Vortex Force Amount=111
    Vortex Origin Offset:Vector3f=&RandomRangeVector(Minimum=<-70 -50 -75>, Maximum=<70 -70 -25>)
    Vortex Origin Offset Coordinate Space:ENiagaraCoordinateSpace=::Local
   )
   &InheritSourceMovement(Applied Position Scale=<0.75 0.75 0.25>, ApplyForce=false, ApplyPosition=true, ApplyPositionFalloff=false)
   &AccelerationForce(
    Acceleration=&FN_FrontEndVector3Modifier(Gameplay=<0 0 0>, Front End=&RandomRangeVector(Minimum=<-666 666 0>, Maximum=<666 3333 666>))
    Coordinate Space=::Local
   )
   &SolveForcesAndVelocity(Clamp Velocity=false, Limit Acceleration=false)
  ]
  rendererClasses [&NiagaraMeshRendererProperties]
  renderers [{rendererClass=&NiagaraMeshRendererProperties, rendererIndex=0}]
 }
 Feather001 {
  bEnabled=false
  emitterUpdateScript [
   &EmitterState(Life Cycle Mode=::NewEnumerator1, Loop Behavior=::NewEnumerator1, Loop Duration Mode=::NewEnumerator1, UseLoopDelay=false)
   &SpawnRate(SpawnRate=20, Use Spawn Probability=false)
   &SpawnBurst_Instantaneous(
    Spawn Count=&RandomRangeInt(Minimum=4, Maximum=6)
    Use Spawn Probability=false
    UseLoopCountLimit=false
   )
   &SpawnBurst_Instantaneous(
    Spawn Count=&RandomRangeInt(Minimum=1, Maximum=2)
    Spawn Time=0.15
    Use Spawn Probability=false
    UseLoopCountLimit=false
   ) {module=SpawnBurst_Instantaneous001}
   &SpawnBurst_Instantaneous(
    Spawn Count=&RandomRangeInt(Minimum=1, Maximum=2)
    Spawn Time=0.3
    Use Spawn Probability=false
    UseLoopCountLimit=false
   ) {module=SpawnBurst_Instantaneous002}
  ]
  particleSpawnScript [
   /SetVariables(Particles.MeshOrientation=&MakeQuatFromAxisAngle(Axis=&RandomRangeVector(Minimum=<0 0 0>, Maximum=<1 1 1>), AngleInDegrees=90)) {isSetParameters=true}
   /SetVariables(Particles.SpawnRotationVector=&RandomRangeVector(Minimum=<0 0 0>, Maximum=<1 1 1>)) {isSetParameters=true, module=SetVariables_#6}
   &InitializeParticle(
    Lifetime=&RandomRangeFloat(Minimum=1.3, Maximum=2.2)
    Mesh Scale=&SplitVector(
     Double Component Vector Channels:ECascadeNiagaraTwoVectorChannels=::NewEnumerator0
     Double Component:NiagaraFloat=&RandomRangeFloat(Minimum=0.3, Maximum=0.5)
     Single Component:NiagaraFloat=&RandomRangeFloat(Minimum=0, Maximum=1)
    )
    Mesh Scale Mode=::NewEnumerator4
    UsePositionOffset=false
    Write Material Random=true
   )
   &ShapeLocation(
    MaskSpawnGroup=false
    Offset=&RandomRangeVector(Minimum=<-170 -70 -30>, Maximum=<170 -30 -20>)
    Offset Coordinate Space=::Local
    Offset Mode=::NewEnumerator0
    Override Seed=false
    Sphere Radius=5
    UseOwnerScale=true
   )
   &AddVelocity(
    Rotation Coordinate Space=::Simulation
    Velocity=&Multiply_VectorByFloat(Vector=$Particles.ShapeLocation.ShapeVector, Float=&RandomRangeFloat(Minimum=5, Maximum=10))
   ) {module=AddVelocity001}
   &AddVelocity(
    Rotation Coordinate Space=::Simulation
    Velocity=&RandomRangeVector(Minimum=<0 0 -150>, Maximum=<0 0 -800>)
   )
   &DynamicMaterialParameters(
    Index 0 Param 1=&RandomRangeFloat(Minimum=0, Maximum=0.9)
    Index 0 Param 4=&RandomRangeFloat(Minimum=0, Maximum=1)
    Module.Param0WriteEnabled=true
    Module.Param1WriteEnabled=false
    Module.Param2WriteEnabled=false
    Module.Param3WriteEnabled=true
   )
  ]
  particleUpdateScript [
   &ParticleState()
   &UpdateMeshOrientation(Rotation Vector=$Particles.SpawnRotationVector)
   &ScaleMeshSize(Scale Factor=&VectorFromFloat(Value=&FloatFromCurve(FloatCurve=DataInterface() {propertyValues {Curve {keys [<at 0.5 1>,<at 1 0>]}}})))
   &Drag(Drag=4, Use Linear Drag=true, Use Rotational Drag=false)
   &VortexForce(
    UseOriginOffset=true
    Vortex Axis=<1 1 1>
    Vortex Axis Coordinate Space:ENiagaraCoordinateSpace=::Local
    Vortex Force Amount=100
    Vortex Origin Offset:Vector3f=&RandomRangeVector(Minimum=<-70 -50 -75>, Maximum=<70 -70 -25>)
    Vortex Origin Offset Coordinate Space:ENiagaraCoordinateSpace=::Local
   )
   &SolveForcesAndVelocity(Clamp Velocity=false, Limit Acceleration=false)
   &DynamicMaterialParameters(
    Index 0 Param 2=&FloatFromCurve(FloatCurve=DataInterface() {propertyValues {Curve {keys [<at 0.6 1 RCIM_Cubic>,<at 1 0 RCIM_Cubic>]}}})
    Module.Param0WriteEnabled=false
    Module.Param1WriteEnabled=true
    Module.Param2WriteEnabled=true
    Module.Param3WriteEnabled=false
   ) {module=DynamicMaterialParameters001}
   &ScaleColor(
    Scale Alpha=&FloatFromCurve(FloatCurve=DataInterface() {propertyValues {Curve {keys [<at 0 1>,<at 0.6 1>,<at 1 0>]}}})
    Scale RGB=&VectorFromCurve(VectorCurve=DataInterface() {propertyValues {XCurve {keys [<at 0 0.6>,<at 0.25 0.5>]}, YCurve {keys [<at 0 0.6>,<at 0.25 0.5>]}, ZCurve {keys [<at 0 0.6>,<at 0.25 0.5>]}}})
    ScaleA=true
    ScaleRGB=true
   )
  ]
  rendererClasses [&NiagaraMeshRendererProperties]
  renderers [{rendererClass=&NiagaraMeshRendererProperties, rendererIndex=0}]
 }
 IDLE_Soft_Trail {
  emitterSpawnScript [&FN_RibbonIDOffsetUtility()]
  emitterUpdateScript [
   &EmitterState()
   &FN_SpawnAnimTrail(Number of Anim Trails=2, Spawn Mode:ENiagaraSkeletalSpawning=::NewEnumerator0)
  ]
  particleSpawnScript [
   &FN_InitializeAnimTrailBoneIndecies()
   &FN_InitializeAnimTrailBoneIndecies(AnimTrail Selection=1, First Bone or Socket Index=2, Second Bone or Socket Index=3) {module=FN_InitializeAnimTrailBoneIndecies001}
   &FN_InitializeAnimTrail(
    AnimTrailWidth=&FN_FrontEndFloatModifier(Gameplay=3.5, Front End=2)
    Color=&SelectLinearColorFromArray(
     Array Sampling Mode=::NewEnumerator1
     Direct Array Index=&MakeIntFromFloat(Float=$User.Variant)
     Color Selection Array=DataInterface() {propertyValues {ColorData [<rgba 0.5 0.375 0.4062 0.25>, <rgba 0.25 0.2387 0.225 0.25>]}}
     Random Seed=?
    )
    Lifetime=&MakeCustomFloatFromBool(Bool=$Particles.Locked, True Float=0, False Float=0.5)
   )
   &AddVelocity(Velocity=&MakeVector(
     X=0
     Y=&FloatFromCurve(
      FloatCurve=DataInterface() {propertyValues {Curve {keys [<at 0 0.25>,<at 0.5 1>]}}}
      CurveIndex=&RemapRange(Input Value To Remap=$User.ForwardDot, Input Min=-1, Input Max=1, Output Min=0, Output Max=1)
      Scale Curve=500
     )
     Z=0
    )) {module=AddVelocity001}
   &AddVelocity(Velocity=&Multiply_VectorByFloat(
     Vector=<0 1500 0>
     Float=&MakeCustomFloatFromBool(Bool=$User.bIsFrontEnd, True Float=-1, False Float=1)
    )) {enabled=false}
  ]
  particleUpdateScript [
   &SolveForcesAndVelocity(Clamp Velocity=false, Limit Acceleration=false)
   &FN_UpdateAnimTrail(SkeletalMesh=$User.SkeletalMesh) {module=FN_UpdateAnimTrail001}
   /SetVariables(Particles.RibbonU0Override=$Particles.NormalizedAge) {isSetParameters=true, module=SetVariables_#7}
   &ScaleColor(
    Scale Alpha=&FloatFromCurve(
     FloatCurve=DataInterface() {propertyValues {Curve {keys [
         {arriveTangent=9.0725, interpMode=RCIM_Cubic, leaveTangent=9.0725, tangentMode=RCTM_User, time=0, value=0}
         {arriveTangent=-2.5249, interpMode=RCIM_Cubic, leaveTangent=-2.5249, tangentMode=RCTM_User, time=0.1408, value=0.9847}
         {arriveTangent=-0.0405, interpMode=RCIM_Cubic, leaveTangent=-0.0405, tangentMode=RCTM_User, time=1, value=0}
        ]}}}
     Scale Curve=&Multiply_Float(
      A=$System.Opacity
      B=&FN_FrontEndFloatModifier(Gameplay=$System.FN_Integrator001.NormalizedValue, Front End=1)
     )
    )
    ScaleA=true
    ScaleRGB=false
   )
  ]
  rendererClasses [&NiagaraRibbonRendererProperties]
  renderers [{rendererClass=&NiagaraRibbonRendererProperties}]
 }
 IDLE_Trail {
  emitterSpawnScript [&FN_RibbonIDOffsetUtility()]
  emitterUpdateScript [
   &EmitterState()
   &FN_SpawnAnimTrail(Number of Anim Trails=2)
  ]
  particleSpawnScript [
   &FN_InitializeAnimTrailBoneIndecies()
   &FN_InitializeAnimTrailBoneIndecies(AnimTrail Selection=1, First Bone or Socket Index=2, Second Bone or Socket Index=3) {module=FN_InitializeAnimTrailBoneIndecies001}
   &FN_InitializeAnimTrail(
    Color=&SelectLinearColorFromArray(
     Array Sampling Mode=::NewEnumerator1
     Direct Array Index=&MakeIntFromFloat(Float=$User.Variant)
     Color Selection Array=DataInterface() {propertyValues {ColorData [<rgba 0.5 0.375 0.4062 0.5>, <rgba 0.666 0.636 0.5994 0.5>]}}
     Random Seed=?
    )
    Lifetime=&MakeCustomFloatFromBool(Bool=$Particles.Locked, True Float=0, False Float=0.15)
   )
   &AddVelocity(Velocity=&Multiply_VectorByFloat(Vector=<0 1500 0>, Float=&MakeCustomFloatFromBool(Bool=$User.bIsFrontEnd, True Float=1, False Float=1)))
  ]
  particleUpdateScript [
   &SolveForcesAndVelocity(Clamp Velocity=false, Limit Acceleration=false)
   &FN_UpdateAnimTrail(SkeletalMesh=$User.SkeletalMesh) {module=FN_UpdateAnimTrail001}
   /SetVariables(Particles.RibbonU0Override=$Particles.NormalizedAge) {isSetParameters=true, module=SetVariables_#7}
   &ScaleColor(
    Scale Alpha=&FloatFromCurve(
     FloatCurve=DataInterface() {propertyValues {Curve {keys [
         {arriveTangent=9.0725, interpMode=RCIM_Cubic, leaveTangent=9.0725, tangentMode=RCTM_User, time=0, value=0}
         {arriveTangent=-2.5249, interpMode=RCIM_Cubic, leaveTangent=-2.5249, tangentMode=RCTM_User, time=0.1408, value=0.9847}
         {arriveTangent=-0.0405, interpMode=RCIM_Cubic, leaveTangent=-0.0405, tangentMode=RCTM_User, time=1, value=0}
        ]}}}
     Scale Curve=&Multiply_Float(
      A=$System.Opacity
      B=&FN_FrontEndFloatModifier(Gameplay=$System.FN_Integrator001.NormalizedValue, Front End=1)
     )
    )
    ScaleA=true
    ScaleRGB=false
   )
  ]
  rendererClasses [&NiagaraRibbonRendererProperties]
  renderers [{rendererClass=&NiagaraRibbonRendererProperties}]
 }
 Speedlines {
  bEnabled=false
  emitterUpdateScript [
   &EmitterState()
   &SpawnRate(
    SpawnRate=&MakeCustomFloatFromBool(Bool=$User.bIsFrontEndPreview, True Float=50, False Float=90)
    Use Spawn Probability=true
   ) {module=SpawnRate001}
  ]
  particleSpawnScript [
   &InitializeParticle(
    Color=&MakeLinearColorFromVectorAndFloat(
     "Vector (RGB)"=<1 1 1>
     "Float (Alpha)"=&Multiply_Float(
      A=&UniformRangedFloat(Minimum=0.1, Maximum=0.18)
      B=&MakeCustomFloatFromBool(Bool=$User.bIsFrontEndPreview, True Float=0.4, False Float=1)
     )
    )
    Lifetime Max=0.16
    Lifetime Min=0.1
    Lifetime Mode=::NewEnumerator1
    Position Offset=&FN_FrontEndVector3Modifier(Gameplay=<0 0 0>, Front End=<0 50 0>)
    Sprite Size Max=<3 10>
    Sprite Size Min=<2 2>
    Sprite Size Mode=::NewEnumerator2
    UsePositionOffset=true
    Write Material Random=true
   )
   &TorusLocation(
    Align Torus With Axis=true
    Handle Radius=0
    Large Radius:NiagaraFloat=&RandomRangeFloat(Minimum=99, Maximum=111)
    MaskSpawnGroup=false
    Module.Override Randomness=false
    Offset=<0 -180 0>
    Torus Axis=<0 1 0>
   )
   &AddVelocityFromPoint(
    Module.Use Falloff Distance=false
    Origin Offset=<0 -555 0>
    Velocity Strength=&FN_FrontEndFloatModifier(Gameplay=600, Front End=600)
   )
   &AddVelocity(
    Velocity=&RandomRangeVector(Minimum=<0 300 0>, Maximum=<1 700 1>)
    Velocity Speed Scale=&FN_FrontEndFloatModifier(Gameplay=1, Front End=0)
   )
  ]
  particleUpdateScript [
   &SolveForcesAndVelocity(Clamp Velocity=false, Limit Acceleration=false)
   &ParticleState()
   &ScaleSpriteSize(Scale Factor=&Vector2DFromCurve(
     Vector2Curve=DataInterface() {propertyValues {XCurve {keys [<at 0 0.5 RCIM_Cubic>,<at 0.5 1 RCIM_Cubic>]}, YCurve {keys [<at 0 0 RCIM_Cubic>,<at 0.5 1 RCIM_Cubic>]}}}
     ScaleCurve=&Vector2DFromFloat(Value=1)
    ))
   &ScaleSpriteSizeBySpeed(Max Scale Factor=<1 50>, Min Scale Factor=<0.01 0.05>, Velocity Threshold=1)
   &ScaleColor(
    Scale Alpha=&FloatFromCurve(
     FloatCurve=DataInterface() {propertyValues {Curve {keys [
         <at 0 0 RCIM_Cubic>
         <at 0.3759 1.0207>
         {arriveTangent=-2.221, interpMode=RCIM_Cubic, leaveTangent=-2.221, tangentMode=RCTM_User, time=1, value=0}
        ]}}}
     Scale Curve=&MakeCustomFloatFromBool(
      Bool=$User.bIsFrontEndPreview
      True Float=&MakeCustomFloatFromBool(Bool=$User.bIsFullyDeployed, True Float=1, False Float=0)
      False Float=&Lerp_Float(A=0, B=1, Alpha=$User.ForwardDot)
     )
    )
    Scale RGB=&VectorFromFloat(Value=1)
    ScaleA=true
    ScaleRGB=true
   )
  ]
  rendererClasses [&NiagaraSpriteRendererProperties]
  renderers [{rendererClass=&NiagaraSpriteRendererProperties}]
 }
}
summary {
 systemName=NS_Glider_BandageNinja
 emitters [
  {bEnabled=false, emitterName=FN_Debug, rendererClasses [&NiagaraSpriteRendererProperties]}
  {bEnabled=false, emitterName=Speedlines, rendererClasses [&NiagaraSpriteRendererProperties]}
  {emitterName=IDLE_Soft_Trail, rendererClasses [&NiagaraRibbonRendererProperties]}
  {emitterName=IDLE_Trail, rendererClasses [&NiagaraRibbonRendererProperties]}
  {emitterName=Feather, rendererClasses [&NiagaraMeshRendererProperties]}
  {bEnabled=false, emitterName=Feather001, rendererClasses [&NiagaraMeshRendererProperties]}
 ]
 userVariables [
  User.bIsBigFlapping {type=NiagaraBool}
  User.bIsFrontEnd {type=NiagaraBool}
  User.bIsFrontEndPreview {type=NiagaraBool}
  User.bIsFullyDeployed {type=NiagaraBool}
  User.blsFlapping {type=NiagaraBool}
  User.FlapCurve {type=NiagaraFloat, defaultValue=0, description=""}
  User.ForwardDot {type=NiagaraFloat, defaultValue=1, description=""}
  User.New Linear Color {type=LinearColor, description="", defaultValue {struct=LinearColor, value=<rgba 0.0052 0.0021 0.0025 10>}}
  User.RibbonIDOffset {type=NiagaraInt32}
  User.RightDot {type=NiagaraFloat, defaultValue=0, description=""}
  User.SkeletalMesh {
   type=NiagaraDataInterfaceSkeletalMesh
   defaultValue {
    struct=NiagaraExt_VariableValue_DataInterface
    value {dataInterface=&self:NiagaraDataInterfaceSkeletalMesh_0}
   }
  }
  User.Variant {type=NiagaraFloat, defaultValue=0, description=""}
  User.Variant_Color {type=LinearColor, description="", defaultValue {struct=LinearColor, value=<rgba 1 1 1 1>}}
 ]
}
